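# Looping mono MetaSound source that plays continuous breathing for the HeavyLamp voice, using HeavyChandelier inhale and exhale waves.
# On play, a seeded random short delay feeds a trigger sequence that alternates between the Inhale and Exhale wave players.
# Either player finishing loops back through a deferred trigger variable to schedule the next breath.
# Both players are summed and scaled by MasterVolume, with concurrency limited to one instance.
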MetaSoundSource /JethroLager/Assets/Audio/VO/HeavyLamp/MSS_HeavyLamp_Breathing_SourceBus
inputs [
 Exhale {type=WaveAsset}
 Inhale {type=WaveAsset}
 MasterVolume {type=Float}
 Seed {type=Int32}
 UE.Source.OnPlay {type=Trigger}
]
interfaces [UE.OutputFormat.Mono:1.0, UE.Source:1.1]
outputs ["UE.OutputFormat.Mono.Audio:0" {type=Audio}]
pages [Default {
  input_defaults {
   Exhale=/JethroLager/Assets/Audio/VO/HeavyLamp/Waves/HeavyChandelier_Exhale
   Inhale=/JethroLager/Assets/Audio/VO/HeavyLamp/Waves/HeavyChandelier_Inhale
   MasterVolume=2
   Seed=0
   UE.Source.OnPlay=null
  }
  nodes [
   n1 = InitVariable.Trigger()
   n2 = VariableDeferredAccessor.Trigger(Variable=InitVariable.Trigger().Variable)
   TriggerAnyTriggerAny2_0 = "TriggerAny.Trigger Any (2)"(In 0=^UE.Source.OnPlay, In 1=@n2.Value)
   n3 = UE.RandomTime(Seed=^Seed, Next=@TriggerAnyTriggerAny2_0.Out) {also {defaults {Max=0.7, Min=0.6}}}
   TriggerDelay_0 = UE.Trigger Delay(In=@n3.On Next, Delay Time=@n3.Value)
   TriggerSequence2_0 = "UE.Trigger Sequence (2)"(In=@TriggerDelay_0.Out)
   WavePlayer_0 = UE.Wave Player:Mono(
    Wave Asset=^Inhale
    Play=VariableAccessor.Trigger(Variable=VariableMutator.Trigger(Variable=@n1.Variable, Value=@TriggerSequence2_0.Out 0).Variable).Value
   )
   n4 = InitVariable.Trigger()
   WavePlayer_1 = UE.Wave Player:Mono(
    Wave Asset=^Exhale
    Play=VariableAccessor.Trigger(Variable=VariableMutator.Trigger(Variable=@n4.Variable, Value=@TriggerSequence2_0.Out 1).Variable).Value
   )
   Multiply_0 = UE.Multiply:Audio by Float(
    AdditionalOperands=^MasterVolume
    PrimaryOperand=UE.Add:Audio(PrimaryOperand=@WavePlayer_0.Out Mono, AdditionalOperands=@WavePlayer_1.Out Mono).Out
   )
   TriggerAnyTriggerAny2_1 = "TriggerAny.Trigger Any (2)"(In 0=@WavePlayer_0.On Finished, In 1=@WavePlayer_1.On Finished)
   VariableMutator.Trigger(Value=@TriggerAnyTriggerAny2_1.Out, Variable=@n2.Variable)
  ]
  links [{dst="@UE.OutputFormat.Mono.Audio:0", src=Multiply_0.Out}]
 }]
properties {
 assetImportData=&self:AssetImportData
 bLooping=true
 bOverrideConcurrency=true
 duration=10000
 priority=100
 sampleRate=48000
 concurrencyOverrides {maxCount=1, resolutionRule=StopLowestPriority}
}
